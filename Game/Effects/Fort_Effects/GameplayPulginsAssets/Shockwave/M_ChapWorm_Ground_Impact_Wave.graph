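# Translucent, unlit, two-sided Niagara sprite and mesh material for a ChapWorm ground-impact shockwave, drawing a ring of smoke.
# Optionally rotated UVs become radial coordinates with optional per-particle random tiling, then pan a smoke texture through dynamic parameters.
# The smoke's green channel indexes a second texture for particle-tinted emissive.
# It also drives opacity through smoothstep, inner and outer sphere masks, particle alpha and depth fade.
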Material /Game/Effects/Fort_Effects/GameplayPulginsAssets/Shockwave/M_ChapWorm_Ground_Impact_Wave
graph {
 properties {
  bUsedWithNiagaraMeshParticles=true
  bUsedWithNiagaraSprites=true
  blendMode=BLEND_Translucent
  pixelDepthOffsetMode=PDOM_Legacy
  shadingModel=MSM_Unlit
  twoSided=true
  usedShadingModels=Unlit
 }
 nodes [
  n1 = ParticleColor()
  n2 = ScalarParameter() {defaultValue=0.4, parameterName=RandomU_Min}
  n3 = ScalarParameter() {defaultValue=1, parameterName=RandomU_Max}
  n4 = ParticleRandom()
  n5 = AppendVector(
   A=LinearInterpolate(B=LinearInterpolate(A=@n2, B=@n3, Alpha=@n4), Alpha=ScalarParameter() {parameterName=UseRandomUTilePerParticle}) {constA=1}
   B=Constant() {r=1}
  )
  n6 = Add(A=TextureCoordinate(), B=Constant2Vector() {g=-0.5, r=-0.5})
  n7 = Constant2Vector()
  n8 = DynamicParameter() {defaultValue=<rgba -1 0.1 0 1>, paramNames [ExplosionPan, DistortAmount, RotateAmount, DistortPanSpeed]}
  n9 = Multiply(A=@n8.RotateAmount) {constB=-1}
  n10 = MaterialFunctionCall(Vector or UVs=LinearInterpolate(
    A=@n6
    B=MaterialFunctionCall(UVs=@n6, Rotation Center=@n7, "Rotation Angle (0-1)"=@n9) {materialFunction=&CustomRotator}.Rotated Values
    Alpha=ScalarParameter() {defaultValue=1, parameterName=UseDistortion}
   ) {constB=0}) {materialFunction=&VectorToRadialValue}
  n11 = DynamicParameter() {defaultValue=<rgba -1 0.1 0 1>, paramNames [ExplosionPan, DistortAmount, RotateAmount, DistortPanSpeed]}
  n12 = TextureSample(UVs=Panner(Coordinate=Multiply(A=@n5, B=@n10.Radial Coordinates), Time=@n11.ExplosionPan) {speedY=-0.25}) {texture=/Game/Effects/Fort_Effects/Textures/Smoke/T_Puffy_Smoke_Trail_01}
  n13 = TextureCoordinate()
  n14 = Constant2Vector() {g=0.5, r=0.5}
  n15 = Constant() {r=0.2}
  n16 = Constant() {r=0.4}
  n17 = Multiply(
   A=MaterialFunctionCall(Alpha=@n12.G, Min=ScalarParameter() {defaultValue=0.1, parameterName=Min}, Max=ScalarParameter() {defaultValue=1, parameterName=Max}) {materialFunction=&SmoothStep}.Result
   B=OneMinus(None=SphereMask(A=@n13, B=@n14, Radius=@n15, Hardness=@n16))
  )
  n18 = SphereMask(A=TextureCoordinate(), B=Constant2Vector() {g=0.5, r=0.5}, Radius=Constant() {r=0.4}, Hardness=Constant() {r=-1})
  DepthFade(Opacity=Multiply(A=@n1.A, B=Multiply(A=@n17, B=@n18)), FadeDistance=ScalarParameter() {defaultValue=50, parameterName=DepthFade})
  MaterialFunctionCall(Emissive=Multiply(A=@n1.RGB, B=TextureSample(UVs=@n12.G) {texture=/Game/Effects/Fort_Effects/Textures/Tile/MotionBlur}.RGB)) {materialFunction=&MF_FX_EmissiveAdjustments}
 ]
}
